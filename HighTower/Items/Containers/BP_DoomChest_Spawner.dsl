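; BP_DoomChest_Spawner is a spawner blueprint that picks which faction chest class to spawn for the HighTower Doom chest.
; On BeginPlay, a scalable-float toggle, SwapToHenchmenChest, selects the generic FactionChest parent class when true or the HighTower Date faction chest otherwise.
; It sets ClassToSpawn before calling the parent BeginPlay, leaving the spawning itself to the parent class.
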(event EventBeginPlay
  (if (ScalableFloat|GetValueasBool (Variables|Default|GetSwapToHenchmenChest))
    (Variables|SpawnerVars|SetClassToSpawn "/Game/Building/ActorBlueprints/Containers/Tiered_Chest_Athena_FactionChest_Parent.Tiered_Chest_Athena_FactionChest_Parent_C")
    (else
      (Variables|SpawnerVars|SetClassToSpawn "/HighTower/Items/Containers/Tiered_Chest_Athena_FactionChest_Date.Tiered_Chest_Athena_FactionChest_Date_C")))
  (|CallParentFunction))

(event Collision|EventActorBeginOverlap
  (params (OtherActor :type Actor%20Object%20Reference)))

(event EventTick
  (params (DeltaSeconds :type Float%20%28single-precision%29)))

(fn ConstructionScript
  (|CallParentFunction))
