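; Construction script for the Date variant of the Athena tiered faction chest, giving its glow a green color scheme.
; It runs the parent construction script, then overrides vector parameters on the materials of the inherited-looking Faction_Crate_Glows component.
; The secondary, warm and main glow colors become shades of green, with the main glow pushed above 1 for an HDR, emissive-style intensity.
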(fn ConstructionScript
  (bind _faction_crate_glows (Variables|Default|GetFaction_Crate_Glows))
  (|CallParentFunction)
  (Rendering|Material|SetVectorParameterValueonMaterials _faction_crate_glows "SecColor" (Math|Conversions|ToVector(LinearColor) "(R=0.265329,G=0.635417,B=0.000000,A=1.000000)"))
  (Rendering|Material|SetVectorParameterValueonMaterials _faction_crate_glows "ColorWarm1" (Math|Conversions|ToVector(LinearColor) "(R=0.243945,G=0.800000,B=0.058146,A=1.000000)"))
  (Rendering|Material|SetVectorParameterValueonMaterials _faction_crate_glows "Glow_Colour" (Math|Conversions|ToVector(LinearColor) "(R=1.357295,G=7.000000,B=0.419296,A=2.000000)")))
